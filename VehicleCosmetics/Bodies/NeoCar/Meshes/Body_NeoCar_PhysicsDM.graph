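# Physics asset for the NeoCar vehicle body cosmetic, giving its skeletal mesh a simplified collision shape.
# A single box on the root bone approximates the whole car, offset forward along X and raised above the bone origin.
# It has no other bodies or constraints, so the car is represented as one piece with no articulated or per-part collision.
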PhysicsAsset /VehicleCosmetics/Bodies/NeoCar/Meshes/Body_NeoCar_PhysicsDM
bodies {root [{shapeType=Box, extentX=444, extentY=216, extentZ=116, shapeName=root_box, center=<20 0 92>}]}
constraints []
